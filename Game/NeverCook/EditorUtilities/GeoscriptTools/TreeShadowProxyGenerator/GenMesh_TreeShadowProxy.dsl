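; Editor Geometry Script actor that generates a reduced tree mesh as a shadow proxy from a source LOD.
; On rebuild, it splits the tree by material and casts BVH rays from tessellated cast-mesh vertices toward that component's origin.
; Canopy triangles never hit first are deleted, then all sections are appended to the dynamic mesh.
; A bake function creates or overwrites a "_Bake" Nanite static mesh and copies the source materials.
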(event Events|EventOnRebuildGeneratedMesh
  (params (TargetMesh :type Dynamic%20Mesh%20Object%20Reference))
  (Variables|Default|SetcurrentTargetMesh TargetMesh)
  (MeshGen|GenerateProxy))

(fn 1_RunGeoScript
  (DynamicMeshActor|MarkforMeshRebuild true true))

(fn 2_Bake
  (Utilities|FlowControl|Sequence
    (:then_0
      (LocalVariables|Setl_naniteSettings (Utilities|Struct|MakeMeshNaniteSettings true false true false true false "None" -2147483648 -1 -1 -1 1.0 0.0 "PlatformDefault" "Auto" 1.0 1.0 0.0 64 0 0.0 0 true))
      (LocalVariables|Setl_assetPath (Helper|GetBakeName))
      (bind _returnvalue (EditorScripting|Asset|DoesAssetExist (LocalVariables|Getl_assetPath) :!class /Script/EditorScriptingUtilities.EditorAssetLibrary))
      (if _returnvalue
        (bind _returnvalue_2 (EditorScripting|Asset|LoadAsset (LocalVariables|Getl_assetPath) :!class /Script/EditorScriptingUtilities.EditorAssetLibrary))
        (bind _asstatic_mesh (Utilities|Casting|CastToStaticMesh _returnvalue_2))
        (bind _output_get (LocalVariables|Setl_mesh _asstatic_mesh))
        (GeometryScript|StaticMesh|CopyMeshtoStaticMesh (Variables|Default|GetcurrentTargetMesh) _output_get (Utilities|Struct|MakeGeometryScriptCopyMeshtoAssetOptions :bEnableRecomputeNormals false :bEnableRecomputeTangents false :bEnableRemoveDegenerates false :BoneHierarchyMismatchHandling "DoNothing" :bRemapBoneIndicesToMatchAsset false :bUseOriginalVertexOrder false :bUseBuildScale true :bReplaceMaterials false :NewMaterialUsage "ReplaceAssetMaterials" :bCleanAssignedMaterials false :GenerateLightmapUVs "MatchTargetLODSetting" :bApplyNaniteSettings true :NaniteSettings "(bEnabled=True,FallbackPercentTriangles=100.000000,FallbackRelativeError=0.000000)" :NewNaniteSettings (LocalVariables|Getl_naniteSettings) :bEmitTransaction true :bDeferMeshPostEditChange false))
        (else
          (bind _returnvalue_3 (GeometryScript|AssetManagement|CreateNewStaticMeshAssetfromMesh (Variables|Default|GetcurrentTargetMesh) (LocalVariables|Getl_assetPath) (Utilities|Struct|MakeGeometryScriptCreateNewStaticMeshAssetOptions false false true (LocalVariables|Getl_naniteSettings) 0.0 false))
            (:Success
              (LocalVariables|Setl_mesh _returnvalue_3))))))
    (:then_1
      ;[1 1 1] Set materials on mesh
      (Utilities|IsValid (LocalVariables|Getl_mesh)
        (:Is%20Valid
          (bind (_materiallist _materialindex) (GeometryScript|StaticMesh|GetSectionMaterialListfromStaticMesh (Variables|MeshGen|GetTreeMesh) (Utilities|Struct|MakeGeometryScriptMeshReadLOD "SourceModel" (Helper|GetSourceLODIndex)))
            (:Success
              (for (_array_element _array_index) _materialindex
                (StaticMesh|SetMaterial (LocalVariables|Getl_mesh) _array_element (Utilities|Array|Get(acopy) _materiallist _array_index))))))))))

(fn CastWithBVH
  (params (TargetMesh :type Dynamic%20Mesh%20Object%20Reference) (castTarget :type Vector) (castSource :type Vector) (BVH :type Dynamic%20Mesh%20BVH%20Cache%20Structure))
  (LocalVariables|Setl_castTarget castTarget)
  (LocalVariables|Setl_sample castSource)
  (LocalVariables|Setl_vector (Math|Vector|Normalize (- (LocalVariables|Getl_castTarget) (LocalVariables|Getl_sample))))
  (bind _output_get (LocalVariables|Setl_origin (- (LocalVariables|Getl_sample) (* (LocalVariables|Getl_vector) 1024.0))))
  (bind _hitresult (GeometryScript|Spatial|FindNearestRayIntersectionwithMesh TargetMesh BVH _output_get (LocalVariables|Getl_vector) (Utilities|Struct|MakeBVHQueryOptions (Math|Vector|Distance(Vector) (LocalVariables|Getl_castTarget) (LocalVariables|Getl_origin))))
    (:Found
      (if (Utilities|Struct|BreakRayHitResult _hitresult)
        (return true (.HitTriangleID (Utilities|Struct|BreakRayHitResult _hitresult)))))))

(fn GenerateProxy
  (Utilities|FlowControl|Sequence
    (:then_0
      (bind (_valid _vectorarray _center) (Raycast|GetCastData))
      (if _valid
        (LocalVariables|Setl_castData _vectorarray)
        (LocalVariables|Setl_castTarget _center)
        (Utilities|FlowControl|Sequence
          (:then_0
            (bind _returnvalue (DynamicMeshActor|AllocateComputeMesh))
            (bind _returnvalue_2 (GeometryScript|StaticMesh|CopyMeshFromStaticMeshwithSectionMaterials (Variables|MeshGen|GetTreeMesh) _returnvalue (Utilities|Struct|MakeGeometryScriptCopyMeshfromAssetOptions true) (Utilities|Struct|MakeGeometryScriptMeshReadLOD "RenderData" (Helper|GetSourceLODIndex)))
              (:Success
                (bind _returnvalue_3 (DynamicMeshActor|GetComputeMeshPool))
                (bind (_componentmeshes _componentmaterialids) (GeometryScript|Decomposition|SplitMeshbyMaterialIDs _returnvalue_2 _returnvalue_3))
                (bind _output_get (LocalVariables|Setl_compMeshes _componentmeshes))
                (for (_array_element _array_index) _output_get
                  (GeometryScript|Materials|ClearMaterialIDs _array_element (Utilities|Array|Get(acopy) _componentmaterialids _array_index)))
                (LocalVariables|Setl_compMeshFoliage (Utilities|Array|Get(acopy) (LocalVariables|Getl_compMeshes) (Helper|GetCanopyMatIndex))))))
          (:then_1
            (bind _outputbvh (GeometryScript|Spatial|BuildBVHForMesh (LocalVariables|Getl_compMeshFoliage)))
            (for _array_element_2 (LocalVariables|Getl_castData)
              (bind (_hit _hit_triangle_id) (Raycast|CastwithBVH (LocalVariables|Getl_compMeshFoliage) (LocalVariables|Getl_castTarget) _array_element_2 _outputbvh))
              (if _hit
                (Utilities|Array|Add (LocalVariables|Getl_hitIDs) _hit_triangle_id))))
          (:then_2
            (bind _triangleidlist (GeometryScript|MeshQueries|GetAllTriangleIDs (LocalVariables|Getl_compMeshFoliage)))
            (bind _indexarray (GeometryScript|ListUtils|ConvertIndexListtoArray _triangleidlist))
            (for _array_element_3 _indexarray
              (if (not (Utilities|Array|ContainsItem (LocalVariables|Getl_hitIDs) _array_element_3))
                (Utilities|Array|Add (LocalVariables|Getl_finalIDs) _array_element_3))))
          (:then_3
            (bind _selection (GeometryScript|MeshSelection|ConvertIndexArraytoMeshSelection (LocalVariables|Getl_compMeshFoliage) (LocalVariables|Getl_finalIDs) "Triangles"))
            (bind (_numdeleted _returnvalue_4) (GeometryScript|MeshEdits|DeleteSelectedTrianglesfromMesh (LocalVariables|Getl_compMeshFoliage) _selection))
            (LocalVariables|Setl_compMeshFoliage _returnvalue_4)
            (Development|PrintString (Utilities|String|Append "Deleted Triangle Count: " (Utilities|String|ToString(Integer) _numdeleted))))
          (:then_4
            (for _array_element_4 (LocalVariables|Getl_compMeshes)
              (GeometryScript|MeshEdits|AppendMesh :TargetMesh (Variables|Default|GetcurrentTargetMesh) :AppendMesh _array_element_4 :bDeferChangeNotifications false :AppendOptions (Utilities|Struct|MakeGeometryScriptAppendMeshOptions "UseSource")))))))
    (:then_1
      (DynamicMeshActor|ReleaseAllComputeMeshes))))

(fn getBakeName
  (return (Utilities|String|Append (Utilities|String|Split (Utilities|BreakSoftClassPath (Utilities|GetSoftClassPath (Utilities|GetClass self))) ".") "_Bake")))

(fn GetCanopyMatIndex
  (return (Variables|MeshGen|GetCanopyMaterialIndex)))

(fn GetCastData
  (Utilities|IsValid (Variables|Default|GetRayCastMeshComponent)
    (:Is%20Valid
      (bind _output_get (LocalVariables|Setl_mesh (Variables|/Script/Engine.StaticMeshComponent|GetStaticMesh (Variables|Default|GetRayCastMeshComponent))))
      (Utilities|IsValid _output_get
        (:Is%20Valid
          (bind _returnvalue (DynamicMeshActor|AllocateComputeMesh))
          (bind _returnvalue_2 (GeometryScript|StaticMesh|CopyMeshFromStaticMeshwithSectionMaterials (LocalVariables|Getl_mesh) _returnvalue (Utilities|Struct|MakeGeometryScriptCopyMeshfromAssetOptions false false false) (Utilities|Struct|MakeGeometryScriptMeshReadLOD "MaxAvailable"))
            (:Success
              (bind _returnvalue_3 (GeometryScript|Transforms|TransformMesh _returnvalue_2 (Transformation|GetWorldTransform (Variables|Default|GetRayCastMeshComponent))))
              (bind _returnvalue_4 (GeometryScript|Subdivide|ApplyUniformTessellation _returnvalue_3 (Variables|MeshGen|GetCastMeshTesselationLevel)))
              (bind (_positionlist _bhasvertexidgaps _returnvalue_5) (GeometryScript|MeshQueries|GetAllVertexPositions _returnvalue_4 true))
              (bind _vectorarray (GeometryScript|ListUtils|ConvertVectorListtoArray _positionlist))
              (DynamicMeshActor|ReleaseComputeMesh _returnvalue_5)
              (return (Utilities|Array|IsValidIndex _vectorarray) _vectorarray (Transformation|GetWorldLocation (Variables|Default|GetRayCastMeshComponent))))))))))

(fn getSourceLODIndex
  (return (Variables|MeshGen|GetLODToBake)))

(fn ConstructionScript)
